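# Niagara spawn effect for the CurbRestore player emote, kicking up debris and dust.
# Debris spawns two color groups, tagged for mesh and sprite renderers, that tumble, fall and shrink, with collision variants gated by quality level and age.
# Dust rises as expanding flipbook sprites slowed by drag, with time-of-day-scaled alpha and fog parameters.
# Alternate debris, dust burst and sword-mesh emitters, laid out across three boxes, are disabled.
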NiagaraSystem /BRCosmetics/Animation/Game/MainPlayer/Emotes/CurbRestore/FX/NS_Emote_CurbRestore_Spawn
dependencies {
 usedDataInterfaces [
  &NiagaraDataInterfaceArrayFloat3
  &NiagaraDataInterfaceCurve
  &NiagaraDataInterfacePlatformSet
  &NiagaraDataInterfaceColorCurve
 ]
 usedDynamicInputs [
  &RandomRangeVector2D
  &ColorFromCurve
  &MultiplyLinearColor
  &MakeVectorFromLinearColorRGB
  &FNTimeOfDayColor
  &RandomRangeLinearColor
  &MakeVector4_FromLinearColor
 ]
 usedModules [
  &SystemState
  &self:SystemSpawnScript.SystemScriptSource.SystemScriptGraph.NiagaraNodeAssignment_1.SetVariables_#1
  &self:SystemSpawnScript.SystemScriptSource.SystemScriptGraph.NiagaraNodeAssignment_0.SetVariables_#2
 ]
}
emitters {
 Debris {
  emitterUpdateScript [
   &EmitterState(Life Cycle Mode=::NewEnumerator1, Loop Behavior=::NewEnumerator1, Loop Delay=0.5, Loop Duration=1.5, UseLoopDelay=true)
   &SpawnRate(Spawn Probability=0.8, SpawnRate=8, Use Spawn Probability=true)
   &SpawnRate(Spawn Probability=0.8, SpawnGroup=1, SpawnRate=6, Use Spawn Probability=true) {module=SpawnRate001}
  ]
  particleSpawnScript [
   &FN_SetSpawnGroup(Number of Spawn Groups=2)
   /SetVariables(Particles.VisibilityTag=$Particles.SpawnGroup) {isSetParameters=true, module=SetVariables_#3}
   &InitializeParticle(
    Color Maximum=&MergeHelper_ColorSwitch(0=<rgba 0.4921 0.5944 0.7083 1>, 1=<rgba 0.375 0.2328 0.1562 0.5>, 2=?, 3=?, Index=$Particles.VisibilityTag, Num=::NewEnumerator0)
    Color Minimum=&MergeHelper_ColorSwitch(0=<rgba 0.5925 0.8193 0.9844 1>, 1=<rgba 0.375 0.2073 0.1562 0.35>, 2=?, 3=?, Index=$Particles.VisibilityTag, Num=::NewEnumerator0)
    Color Mode=::NewEnumerator2
    Lifetime Max=1
    Lifetime Min=0.5
    Lifetime Mode=::NewEnumerator1
    Mass Mode=::NewEnumerator1
    Mesh Scale Max=<1 1 1>
    Mesh Scale Min=<0.5 0.5 0.5>
    Mesh Scale Mode=::NewEnumerator2
    Position Offset Coordinate Space=::Local
    Sprite Rotation Mode=::NewEnumerator2
    Sprite Size Mode=::NewEnumerator1
    Sprite UV Mode=::NewEnumerator3
    Uniform Sprite Size Max=6
    Uniform Sprite Size Min=3
    UsePositionOffset=true
    Write Material Random=true
   )
   &ShapeLocation(MaskSpawnGroup=false, Override Seed=false, Sphere Radius=60, UseOwnerScale=true)
   &AddVelocity(
    Module.Use Falloff Distance=false
    Override Seed=false
    Velocity Mode=::NewEnumerator1
    Velocity Speed=&RandomRangeFloat(Minimum=100, Maximum=200)
   )
   &AddVelocity(Velocity=&RandomRangeVector(Minimum=<0 0 100>, Maximum=<0 0 200>)) {module=AddVelocity001}
   /SetVariables(Particles.MeshIndex=&RandomRangeInt(Minimum=0, Maximum=2)) {isSetParameters=true, module=SetVariables_#4}
   &InitialMeshOrientation(Use Rotation Vector=false)
  ]
  particleUpdateScript [
   &ParticleState()
   &UpdateMeshOrientation(
    Rotation Rate=&RandomRangeFloat(Minimum=1, Maximum=2)
    Rotation Vector=&RandomRangeVector(Minimum=<-1 -1 -1>, Maximum=<1 1 1>)
   )
   &GravityForce(Gravity=&RandomRangeVector(Minimum=<0 0 -680>, Maximum=<1 1 -980>))
   &Drag(Use Linear Drag=true, Use Rotational Drag=true)
   &Collision(
    Age Colliding Particles=false
    Collision Enabled=&Boolean_AndOperation(
     Bool1=&IsPlatformSetActive(New Platform Set=DataInterface() {propertyValues {Platforms {qualityLevelMask=-8}}})
     Bool2=&SetBoolByFloatComparison(
      A=&FloatFromCurve(FloatCurve=DataInterface() {propertyValues {Curve {
          defaultValue=1
          keys [{interpMode=RCIM_Cubic, tangentMode=RCTM_User, time=0, value=0}, {interpMode=RCIM_Cubic, tangentMode=RCTM_User, time=1, value=1}]
         }}})
      B=0.5
     )
    )
    KillOnCollision=false
    Module.Control Rest State Via Surface Normal=false
    Module.EnableMaxCollisionCount=false
    Module.Randomize Collision Normal Vector=false
    Radius Calculation Type=::NewEnumerator1
   )
   &Collision(
    Age Colliding Particles=false
    Analytical Collision Plane Position 1:Vector3f=<0 0 -20>
    Analytical Collision Plane Space:ENiagaraCoordinateSpace=::Local
    CPU Collision Type=::NewEnumerator3
    Collision Enabled=&Boolean_AndOperation(
     Bool1=&IsPlatformSetActive(New Platform Set=DataInterface() {propertyValues {Platforms {qualityLevelMask=-28}}})
     Bool2=&SetBoolByFloatComparison(
      A=&FloatFromCurve(FloatCurve=DataInterface() {propertyValues {Curve {
          defaultValue=1
          keys [{interpMode=RCIM_Cubic, tangentMode=RCTM_User, time=0, value=0}, {interpMode=RCIM_Cubic, tangentMode=RCTM_User, time=1, value=1}]
         }}})
      B=0.5
     )
    )
    KillOnCollision=false
    Module.Control Rest State Via Surface Normal=false
    Module.EnableMaxCollisionCount=false
    Module.Randomize Collision Normal Vector=false
    Perform 2 Analytical Collision Queries=false
   ) {module=Collision001}
   &SolveForcesAndVelocity(Clamp Velocity=false, Limit Acceleration=false)
   &ScaleSpriteSize(Scale Factor=&Vector2DFromFloat(Value=&FloatFromCurve(FloatCurve=DataInterface() {propertyValues {Curve {keys [<at 0.5 1 RCIM_Cubic>, {arriveTangent=-3.7562, interpMode=RCIM_Cubic, leaveTangent=-3.7562, tangentMode=RCTM_User, time=1, value=0}]}}})))
   &ScaleMeshSize(Scale Factor=&VectorFromFloat(Value=&FloatFromCurve(FloatCurve=DataInterface() {propertyValues {Curve {keys [
          {arriveTangent=-0.008, interpMode=RCIM_Cubic, leaveTangent=-0.008, tangentMode=RCTM_User, time=0, value=1}
          {arriveTangent=-3.1748, interpMode=RCIM_Cubic, leaveTangent=-3.1748, tangentMode=RCTM_User, time=0.9, value=0}
         ]}}})))
  ]
  rendererClasses [&NiagaraMeshRendererProperties, &NiagaraSpriteRendererProperties]
  renderers [
   {rendererClass=&NiagaraMeshRendererProperties, rendererIndex=0}
   {rendererClass=&NiagaraSpriteRendererProperties, rendererIndex=1}
  ]
 }
 Debris001 {
  bEnabled=false
  emitterUpdateScript [
   &EmitterState(Life Cycle Mode=::NewEnumerator1, Loop Behavior=::NewEnumerator1, Loop Delay=0.5, Loop Duration=3, UseLoopDelay=true)
   &SpawnRate(SpawnGroup=1, SpawnRate=5, Use Spawn Probability=false) {module=SpawnRate001}
   &SpawnRate(SpawnRate=5, Use Spawn Probability=false)
   &SpawnBurst_Instantaneous(Spawn Count=15, Use Spawn Probability=false, UseLoopCountLimit=false) {enabled=false}
  ]
  particleSpawnScript [
   &FN_SetSpawnGroup(Number of Spawn Groups=2)
   &PartitionParticles(Number Of Partitions=3, Particles Per Partition=10, Partition Mode:ENiagara_PartitionMode_v2=::NewEnumerator1) {enabled=false}
   &PartitionParticles(Number Of Partitions=3, Partition Mode:ENiagara_PartitionMode_v2=::NewEnumerator5) {module=PartitionParticles001}
   /SetVariables(Particles.VisibilityTag=$Particles.SpawnGroup) {isSetParameters=true, module=SetVariables_#3}
   &InitializeParticle(
    Color Maximum=&MergeHelper_ColorSwitch(0=<rgba 0.4921 0.5944 0.7083 1>, 1=<rgba 0.375 0.2328 0.1562 0.5>, 2=?, 3=?, Index=$Particles.VisibilityTag, Num=::NewEnumerator0)
    Color Minimum=&MergeHelper_ColorSwitch(0=<rgba 0.5925 0.8193 0.9844 1>, 1=<rgba 0.375 0.2073 0.1562 0.35>, 2=?, 3=?, Index=$Particles.VisibilityTag, Num=::NewEnumerator0)
    Color Mode=::NewEnumerator2
    Lifetime Max=1.5
    Lifetime Mode=::NewEnumerator1
    Mass Mode=::NewEnumerator1
    Mesh Scale Max=<1 1 1>
    Mesh Scale Min=<0.5 0.5 0.5>
    Mesh Scale Mode=::NewEnumerator2
    Position Offset Coordinate Space=::Local
    Sprite Rotation Mode=::NewEnumerator2
    Sprite Size Mode=::NewEnumerator1
    Sprite UV Mode=::NewEnumerator3
    Uniform Sprite Size Max=6
    Uniform Sprite Size Min=3
    UsePositionOffset=true
    Write Material Random=true
   )
   &ShapeLocation(
    Apply To Particle Position=::NewEnumerator1
    Box Size=<25 90 1>
    MaskSpawnGroup=false
    Offset=<-52 0 0>
    Offset Mode=::NewEnumerator0
    Override Seed=false
    Rotation Mode=::NewEnumerator2
    Shape Primitive=::NewEnumerator2
    Surface Only=false
    UseOwnerScale=true
    Yaw / Pitch / Roll=<0.0427 0 0>
   )
   &ShapeLocation(
    Apply To Particle Position=::NewEnumerator1
    Box Size=<70 15 1>
    MaskSpawnGroup=false
    Offset=<0 -50 0>
    Offset Mode=::NewEnumerator0
    Override Seed=false
    Shape Primitive=::NewEnumerator2
    Surface Only=false
    UseOwnerScale=true
   ) {module=ShapeLocation001}
   &ShapeLocation(
    Apply To Particle Position=::NewEnumerator1
    Box Size=<25 90 1>
    MaskSpawnGroup=false
    Offset=<52 0 0>
    Offset Mode=::NewEnumerator0
    Override Seed=false
    Rotation Mode=::NewEnumerator2
    Shape Primitive=::NewEnumerator2
    Surface Only=false
    UseOwnerScale=true
    Yaw / Pitch / Roll=<-0.0427 0 0>
   ) {module=ShapeLocation002}
   &SystemLocation(
    MaskSpawnGroup=false
    Offset=&MergeHelper_VectorSwitch(
     0=Linked() {linkedVariable {name=Particles.ShapeLocation.ShapePosition, type=NiagaraPosition}}
     1=Linked() {linkedVariable {name=Particles.ShapeLocation001.ShapePosition, type=NiagaraPosition}}
     2=Linked() {linkedVariable {name=Particles.ShapeLocation002.ShapePosition, type=NiagaraPosition}}
     3=?
     Index=$Particles.PartitionParticles001.Partition
     Num=::NewEnumerator1
    )
   )
   &AddVelocity(
    Module.Use Falloff Distance=false
    Origin Offset=<0 0 1>
    Override Seed=false
    Velocity Mode=::NewEnumerator1
    Velocity Speed=&RandomRangeFloat(Minimum=100, Maximum=200)
   )
   &AddVelocity(Velocity=&RandomRangeVector(Minimum=<0 0 150>, Maximum=<0 0 200>)) {module=AddVelocity001}
   /SetVariables(Particles.MeshIndex=&RandomRangeInt(Minimum=0, Maximum=2)) {isSetParameters=true, module=SetVariables_#4}
   &InitialMeshOrientation(Use Rotation Vector=false)
  ]
  particleUpdateScript [
   &ParticleState()
   &UpdateMeshOrientation(
    Rotation Rate=&RandomRangeFloat(Minimum=1, Maximum=2)
    Rotation Vector=&RandomRangeVector(Minimum=<-1 -1 -1>, Maximum=<1 1 1>)
   )
   &GravityForce(Gravity=&RandomRangeVector(Minimum=<0 0 -680>, Maximum=<1 1 -980>))
   &Drag(
    Drag=&FloatFromCurve(
     FloatCurve=DataInterface() {propertyValues {Curve {
        defaultValue=1
        keys [{interpMode=RCIM_Cubic, tangentMode=RCTM_User, time=0, value=0}, {interpMode=RCIM_Cubic, tangentMode=RCTM_User, time=0.5781, value=1}]
       }}}
     Scale Curve=&RandomRangeFloat(Minimum=3, Maximum=5)
    )
    Use Linear Drag=true
    Use Rotational Drag=true
   )
   &Collision(
    Age Colliding Particles=false
    Collision Enabled=&Boolean_AndOperation(
     Bool1=&IsPlatformSetActive(New Platform Set=DataInterface() {propertyValues {Platforms {qualityLevelMask=-8}}})
     Bool2=&SetBoolByFloatComparison(
      A=&FloatFromCurve(FloatCurve=DataInterface() {propertyValues {Curve {
          defaultValue=1
          keys [{interpMode=RCIM_Cubic, tangentMode=RCTM_User, time=0, value=0}, {interpMode=RCIM_Cubic, tangentMode=RCTM_User, time=1, value=1}]
         }}})
      B=0.5
     )
    )
    KillOnCollision=false
    Module.Control Rest State Via Surface Normal=false
    Module.EnableMaxCollisionCount=false
    Module.Randomize Collision Normal Vector=false
    Radius Calculation Type=::NewEnumerator1
   )
   &Collision(
    Age Colliding Particles=false
    Analytical Collision Plane Position 1:Vector3f=<0 0 -20>
    Analytical Collision Plane Space:ENiagaraCoordinateSpace=::Local
    CPU Collision Type=::NewEnumerator3
    Collision Enabled=&Boolean_AndOperation(
     Bool1=&IsPlatformSetActive(New Platform Set=DataInterface() {propertyValues {Platforms {qualityLevelMask=-28}}})
     Bool2=&SetBoolByFloatComparison(
      A=&FloatFromCurve(FloatCurve=DataInterface() {propertyValues {Curve {
          defaultValue=1
          keys [{interpMode=RCIM_Cubic, tangentMode=RCTM_User, time=0, value=0}, {interpMode=RCIM_Cubic, tangentMode=RCTM_User, time=1, value=1}]
         }}})
      B=0.5
     )
    )
    KillOnCollision=false
    Module.Control Rest State Via Surface Normal=false
    Module.EnableMaxCollisionCount=false
    Module.Randomize Collision Normal Vector=false
    Perform 2 Analytical Collision Queries=false
   ) {module=Collision001}
   &SolveForcesAndVelocity(Clamp Velocity=false, Limit Acceleration=false)
   &ScaleSpriteSize(Scale Factor=&Vector2DFromFloat(Value=&FloatFromCurve(FloatCurve=DataInterface() {propertyValues {Curve {keys [<at 0.5 1 RCIM_Cubic>, {arriveTangent=-3.7562, interpMode=RCIM_Cubic, leaveTangent=-3.7562, tangentMode=RCTM_User, time=1, value=0}]}}})))
   &ScaleMeshSize(Scale Factor=&VectorFromFloat(Value=&FloatFromCurve(FloatCurve=DataInterface() {propertyValues {Curve {keys [
          {arriveTangent=-0.008, interpMode=RCIM_Cubic, leaveTangent=-0.008, tangentMode=RCTM_User, time=0, value=1}
          {arriveTangent=-3.1748, interpMode=RCIM_Cubic, leaveTangent=-3.1748, tangentMode=RCTM_User, time=0.9, value=0}
         ]}}})))
  ]
  rendererClasses [&NiagaraMeshRendererProperties, &NiagaraSpriteRendererProperties]
  renderers [
   {rendererClass=&NiagaraMeshRendererProperties, rendererIndex=0}
   {rendererClass=&NiagaraSpriteRendererProperties, rendererIndex=1}
  ]
 }
 Dust {
  emitterUpdateScript [
   &EmitterState(
    Inactive Response=::NewEnumerator2
    Life Cycle Mode=::NewEnumerator1
    Loop Behavior=::NewEnumerator1
    Loop Delay=0.2
    Loop Duration=2.8
    UseLoopDelay=true
   )
   &SpawnRate(SpawnGroup=1, SpawnRate=7, Use Spawn Probability=false) {module=SpawnRate001}
   &SpawnRate(SpawnRate=7, Use Spawn Probability=false)
  ]
  particleSpawnScript [
   &InitializeParticle(
    Color=&Multiply_LinearColorByFloat(Linear Color=$System.Global color, Float=1)
    Lifetime=&RandomRangeFloat(Minimum=0.5, Maximum=1)
    Sprite Rotation Angle=&RandomRangeFloat(Minimum=0, Maximum=360)
    Sprite Rotation Mode=::NewEnumerator0
    Sprite Size Mode=::NewEnumerator3
    Uniform Sprite Size=&RandomRangeFloat(Minimum=20, Maximum=15)
    UsePositionOffset=false
    Write Material Random=true
   )
   &ShapeLocation(MaskSpawnGroup=false, Non Uniform Scale=<1 1 0.1>, Offset Mode=::NewEnumerator0, Override Seed=false, Sphere Radius=25, UseOwnerScale=true)
   &AddVelocity(Velocity=<0 0 200>)
   &AddVelocity(
    Module.Use Falloff Distance=false
    Override Seed=false
    Velocity Mode=::NewEnumerator1
    Velocity Speed=&RandomRangeFloat(Minimum=80, Maximum=150)
   ) {module=AddVelocity001}
  ]
  particleUpdateScript [
   &ParticleState()
   &ScaleSpriteSize(
    Uniform Curve Scale=10
    Uniform Curve Sprite Scale=DataInterface() {propertyValues {Curve {defaultValue=1, keys [{tangentMode=RCTM_User, time=0, value=0.7}, {tangentMode=RCTM_User, time=1, value=1}]}}}
   )
   &SubUVAnimation(
    End Frame=15
    SubUV Animation Mode:ENiagara_SubUVLookupMode=::NewEnumerator1
    SubUV Curve=DataInterface() {propertyValues {Curve {keys [{tangentMode=RCTM_User, time=0, value=0}, {tangentMode=RCTM_User, time=1.0138, value=0.388}]}}}
   )
   &Drag(
    Drag=&FloatFromCurve(
     FloatCurve=DataInterface() {propertyValues {Curve {
        defaultValue=1
        keys [{interpMode=RCIM_Cubic, tangentMode=RCTM_User, time=0, value=0}, {interpMode=RCIM_Cubic, tangentMode=RCTM_User, time=0.3793, value=1}]
       }}}
     Scale Curve=7
    )
    Use Linear Drag=true
    Use Rotational Drag=true
   )
   &SolveForcesAndVelocity(Clamp Velocity=false, Limit Acceleration=false)
   /SetVariables(
    Particles.0 - Env Fog Scale Color:LinearColor=&MakeLinearColorFromVectorAndFloat(
     "Vector (RGB)"=<1 1 1>
     "Float (Alpha)"=&FloatFromCurve(
      FloatCurve=DataInterface() {propertyValues {Curve {keys [
          {arriveTangent=3.5962, interpMode=RCIM_Cubic, leaveTangent=3.5962, tangentMode=RCTM_User, time=0, value=0}
          {
           arriveTangent=0.1817
           arriveTangentWeight=0.1842
           interpMode=RCIM_Cubic
           leaveTangent=0.1817
           leaveTangentWeight=0.7369
           tangentMode=RCTM_User
           time=0.5037
           value=0.7647
          }
          {arriveTangent=-4.7233, tangentMode=RCTM_User, time=1, value=0}
         ]}}}
      Scale Curve=&RandomRangeFloat(
       Minimum=0.02
       Maximum=&FNTimeOfDayScalar(Night Scale=0.05, Morning Scale=0.05, Day Scale=0.07, Evening Scale=0.05)
      )
     )
    )
    Particles.1 - Ground Fog Scale Color:LinearColor=&MakeLinearColorFromVectorAndFloat(
     "Vector (RGB)"=<1 1 1>
     "Float (Alpha)"=&FloatFromCurve(
      FloatCurve=DataInterface() {propertyValues {Curve {keys [
          {arriveTangent=5, interpMode=RCIM_Cubic, time=0, value=0}
          {arriveTangent=1.5181, interpMode=RCIM_Cubic, leaveTangent=1.5181, time=0.2, value=0.7}
          {
           arriveTangent=0.1817
           arriveTangentWeight=0.1842
           interpMode=RCIM_Cubic
           leaveTangent=0.1817
           leaveTangentWeight=0.7369
           tangentMode=RCTM_User
           time=0.5037
           value=0.7647
          }
          <at 1 0>
         ]}}}
      Scale Curve=&RandomRangeFloat(
       Minimum=&FNTimeOfDayScalar(Night Scale=0.035, Morning Scale=0.07, Day Scale=0.05, Evening Scale=0.07)
       Maximum=&FNTimeOfDayScalar(Night Scale=0.05, Morning Scale=0.07, Day Scale=0.07, Evening Scale=0.1)
      )
     )
    )
   ) {isSetParameters=true, module=SetVariables_#5}
   /SetVariables(
    Particles.0 - Env Fog Dyn Param:Vector4f=&MakeVector4(X=0, Y=1, Z=1, W=1)
    Particles.1 - Ground Fog Dyn Param:Vector4f=&MakeVector4(X=0, Y=1, Z=0.1, W=1)
   ) {isSetParameters=true, module=SetVariables_#6}
   &ScaleColor(
    Scale Alpha=&FloatFromCurve(
     FloatCurve=DataInterface() {propertyValues {Curve {keys [
         {arriveTangent=6.4271, interpMode=RCIM_Cubic, leaveTangent=6.4271, tangentMode=RCTM_User, time=0, value=0}
         {arriveTangent=0.5834, interpMode=RCIM_Cubic, leaveTangent=0.5834, tangentMode=RCTM_User, time=0.0598, value=0.6812}
         {
          arriveTangent=0.1817
          arriveTangentWeight=0.1842
          interpMode=RCIM_Cubic
          leaveTangent=0.1817
          leaveTangentWeight=0.7369
          tangentMode=RCTM_User
          time=0.5037
          value=0.8
         }
         <at 1 0>
        ]}}}
     Scale Curve=&RandomRangeFloat(
      Minimum=&FNTimeOfDayScalar(Night Scale=0.035, Morning Scale=0.03, Day Scale=0.05, Evening Scale=0.04)
      Maximum=&FNTimeOfDayScalar(Night Scale=0.05, Morning Scale=0.05, Day Scale=0.07, Evening Scale=0.07)
     )
    )
    ScaleA=true
    ScaleRGB=true
   )
  ]
  rendererClasses [&NiagaraSpriteRendererProperties]
  renderers [{rendererClass=&NiagaraSpriteRendererProperties}]
 }
 DustBurst {
  bEnabled=false
  emitterUpdateScript [
   &EmitterState(
    Inactive Response=::NewEnumerator2
    Life Cycle Mode=::NewEnumerator1
    Loop Behavior=::NewEnumerator1
    Loop Delay=0.5
    Loop Duration=2
    UseLoopDelay=true
   )
   &SpawnRate(SpawnRate=30, Use Spawn Probability=false)
   &SpawnBurst_Instantaneous(Spawn Count=30, Use Spawn Probability=false, UseLoopCountLimit=false) {enabled=false}
  ]
  particleSpawnScript [
   &FN_SetSpawnGroup(Number of Spawn Groups=3)
   &InitializeParticle(
    Color=$System.Global color
    Lifetime=&RandomRangeFloat(Minimum=0.4, Maximum=1.2)
    Sprite Rotation Angle=&RandomRangeFloat(Minimum=0, Maximum=360)
    Sprite Rotation Mode=::NewEnumerator0
    Sprite Size Mode=::NewEnumerator3
    Uniform Sprite Size=&RandomRangeFloat(Minimum=20, Maximum=10)
    UsePositionOffset=false
    Write Material Random=true
   )
   &ShapeLocation(
    Apply To Particle Position=::NewEnumerator1
    Box Size=<25 90 1>
    MaskSpawnGroup=false
    Offset=<-52 0 0>
    Offset Mode=::NewEnumerator0
    Override Seed=false
    Rotation Mode=::NewEnumerator2
    Shape Primitive=::NewEnumerator2
    Surface Only=false
    UseOwnerScale=true
    Yaw / Pitch / Roll=<0.0427 0 0>
   )
   &ShapeLocation(
    Apply To Particle Position=::NewEnumerator1
    Box Size=<70 15 1>
    MaskSpawnGroup=false
    Offset=<0 -50 0>
    Offset Mode=::NewEnumerator0
    Override Seed=false
    Shape Primitive=::NewEnumerator2
    Surface Only=false
    UseOwnerScale=true
   ) {module=ShapeLocation001}
   &ShapeLocation(
    Apply To Particle Position=::NewEnumerator1
    Box Size=<25 90 1>
    MaskSpawnGroup=false
    Offset=<52 0 0>
    Offset Mode=::NewEnumerator0
    Override Seed=false
    Rotation Mode=::NewEnumerator2
    Shape Primitive=::NewEnumerator2
    Surface Only=false
    UseOwnerScale=true
    Yaw / Pitch / Roll=<-0.0427 0 0>
   ) {module=ShapeLocation002}
   &SystemLocation(
    MaskSpawnGroup=false
    Offset=&MergeHelper_VectorSwitch(
     0=Linked() {linkedVariable {name=Particles.ShapeLocation.ShapePosition, type=NiagaraPosition}}
     1=Linked() {linkedVariable {name=Particles.ShapeLocation001.ShapePosition, type=NiagaraPosition}}
     2=Linked() {linkedVariable {name=Particles.ShapeLocation002.ShapePosition, type=NiagaraPosition}}
     3=?
     Index=&SpawnGroup()
     Num=::NewEnumerator1
    )
   )
   &AddVelocity(Velocity=<0 0 200>, Velocity Speed Scale=&RandomRangeFloat(Minimum=0.5, Maximum=1))
   &AddVelocity(
    Module.Use Falloff Distance=false
    Override Seed=false
    Velocity Mode=::NewEnumerator1
    Velocity Speed=&RandomRangeFloat(Minimum=20, Maximum=170)
   ) {module=AddVelocity001}
  ]
  particleUpdateScript [
   &ParticleState()
   &ScaleSpriteSize(
    Uniform Curve Scale=7
    Uniform Curve Sprite Scale=DataInterface() {propertyValues {Curve {defaultValue=1, keys [{tangentMode=RCTM_User, time=0, value=0.7}, {tangentMode=RCTM_User, time=1, value=1}]}}}
   )
   &SubUVAnimation(
    End Frame=15
    SubUV Animation Mode:ENiagara_SubUVLookupMode=::NewEnumerator1
    SubUV Curve=DataInterface() {propertyValues {Curve {keys [{tangentMode=RCTM_User, time=0, value=0}, {tangentMode=RCTM_User, time=1.0138, value=0.388}]}}}
   )
   &Drag(
    Drag=&FloatFromCurve(
     FloatCurve=DataInterface() {propertyValues {Curve {
        defaultValue=1
        keys [{interpMode=RCIM_Cubic, tangentMode=RCTM_User, time=0, value=0}, {interpMode=RCIM_Cubic, tangentMode=RCTM_User, time=0.3793, value=1}]
       }}}
     Scale Curve=&RandomRangeFloat(Minimum=4, Maximum=7)
    )
    Use Linear Drag=true
    Use Rotational Drag=true
   )
   &SolveForcesAndVelocity(Clamp Velocity=false, Limit Acceleration=false)
   /SetVariables(
    Particles.0 - Env Fog Scale Color:LinearColor=&MakeLinearColorFromVectorAndFloat(
     "Vector (RGB)"=<1 1 1>
     "Float (Alpha)"=&FloatFromCurve(
      FloatCurve=DataInterface() {propertyValues {Curve {keys [
          {arriveTangent=3.5962, interpMode=RCIM_Cubic, leaveTangent=3.5962, tangentMode=RCTM_User, time=0, value=0}
          {
           arriveTangent=0.1817
           arriveTangentWeight=0.1842
           interpMode=RCIM_Cubic
           leaveTangent=0.1817
           leaveTangentWeight=0.7369
           tangentMode=RCTM_User
           time=0.5037
           value=0.7647
          }
          {arriveTangent=-4.7233, tangentMode=RCTM_User, time=1, value=0}
         ]}}}
      Scale Curve=&RandomRangeFloat(
       Minimum=0.02
       Maximum=&FNTimeOfDayScalar(Night Scale=0.05, Morning Scale=0.05, Day Scale=0.07, Evening Scale=0.05)
      )
     )
    )
    Particles.1 - Ground Fog Scale Color:LinearColor=&MakeLinearColorFromVectorAndFloat(
     "Vector (RGB)"=<1 1 1>
     "Float (Alpha)"=&FloatFromCurve(
      FloatCurve=DataInterface() {propertyValues {Curve {keys [
          {arriveTangent=5, interpMode=RCIM_Cubic, time=0, value=0}
          {arriveTangent=1.5181, interpMode=RCIM_Cubic, leaveTangent=1.5181, time=0.2, value=0.7}
          {
           arriveTangent=0.1817
           arriveTangentWeight=0.1842
           interpMode=RCIM_Cubic
           leaveTangent=0.1817
           leaveTangentWeight=0.7369
           tangentMode=RCTM_User
           time=0.5037
           value=0.7647
          }
          <at 1 0>
         ]}}}
      Scale Curve=&RandomRangeFloat(
       Minimum=&FNTimeOfDayScalar(Night Scale=0.035, Morning Scale=0.07, Day Scale=0.05, Evening Scale=0.07)
       Maximum=&FNTimeOfDayScalar(Night Scale=0.05, Morning Scale=0.07, Day Scale=0.07, Evening Scale=0.1)
      )
     )
    )
   ) {isSetParameters=true, module=SetVariables_#5}
   /SetVariables(
    Particles.0 - Env Fog Dyn Param:Vector4f=&MakeVector4(X=0, Y=1, Z=1, W=1)
    Particles.1 - Ground Fog Dyn Param:Vector4f=&MakeVector4(X=0, Y=1, Z=0.1, W=1)
   ) {isSetParameters=true, module=SetVariables_#6}
   &ScaleColor(
    Scale Alpha=&FloatFromCurve(
     FloatCurve=DataInterface() {propertyValues {Curve {keys [
         {arriveTangent=6.4271, interpMode=RCIM_Cubic, leaveTangent=6.4271, tangentMode=RCTM_User, time=0, value=0.418}
         {arriveTangent=0.5834, interpMode=RCIM_Cubic, leaveTangent=0.5834, tangentMode=RCTM_User, time=0.0598, value=0.6812}
         {
          arriveTangent=0.1817
          arriveTangentWeight=0.1842
          interpMode=RCIM_Cubic
          leaveTangent=0.1817
          leaveTangentWeight=0.7369
          tangentMode=RCTM_User
          time=0.5037
          value=0.8
         }
         <at 1 0>
        ]}}}
     Scale Curve=&RandomRangeFloat(
      Minimum=&FNTimeOfDayScalar(Night Scale=0.035, Morning Scale=0.03, Day Scale=0.05, Evening Scale=0.04)
      Maximum=&FNTimeOfDayScalar(Night Scale=0.05, Morning Scale=0.05, Day Scale=0.07, Evening Scale=0.07)
     )
    )
    ScaleA=true
    ScaleRGB=true
   )
  ]
  rendererClasses [&NiagaraSpriteRendererProperties]
  renderers [{rendererClass=&NiagaraSpriteRendererProperties}]
 }
 Swords {
  bEnabled=false
  emitterUpdateScript [
   &EmitterState(
    Inactive Response=::NewEnumerator1
    Life Cycle Mode=::NewEnumerator1
    Loop Behavior=::NewEnumerator1
    Loop Delay=0.5
    Loop Duration Mode=::NewEnumerator1
    UseLoopDelay=true
   )
   &SpawnBurst_Instantaneous(Spawn Count=14, Spawn Time=0.1, Use Spawn Probability=false, UseLoopCountLimit=false) {enabled=false, module=SpawnBurst_Instantaneous001}
   &SpawnBurst_Instantaneous(Spawn Count=26, Use Spawn Probability=false, UseLoopCountLimit=false)
  ]
  particleSpawnScript [
   &FN_SetSpawnGroup(Number of Spawn Groups=3)
   &InitializeParticle(Color Mode=::NewEnumerator0, Mesh Scale Max=<1 1 0.2>, Mesh Scale Mode=::NewEnumerator2, UsePositionOffset=false, Write Material Random=true)
   &ShapeLocation(
    Apply To Particle Position=::NewEnumerator1
    Box Size=<20 70 1>
    MaskSpawnGroup=false
    Offset=<-52 -5 0>
    Offset Mode=::NewEnumerator0
    Override Seed=false
    Rotation Mode=::NewEnumerator2
    Shape Primitive=::NewEnumerator2
    Surface Only=false
    UseOwnerScale=true
    Yaw / Pitch / Roll=<0.0427 0 0>
   )
   &ShapeLocation(
    Apply To Particle Position=::NewEnumerator1
    Box Size=<70 15 1>
    MaskSpawnGroup=false
    Offset=<0 -50 0>
    Offset Mode=::NewEnumerator0
    Override Seed=false
    Shape Primitive=::NewEnumerator2
    Surface Only=false
    UseOwnerScale=true
   ) {module=ShapeLocation001}
   &ShapeLocation(
    Apply To Particle Position=::NewEnumerator1
    Box Size=<18 60 1>
    MaskSpawnGroup=false
    Offset=<50 -5 0>
    Offset Mode=::NewEnumerator0
    Override Seed=false
    Rotation Mode=::NewEnumerator2
    Shape Primitive=::NewEnumerator2
    Surface Only=false
    UseOwnerScale=true
    Yaw / Pitch / Roll=<-0.0427 0 0>
   ) {module=ShapeLocation002}
   &SystemLocation(
    MaskSpawnGroup=false
    Offset=&Add_Vector(
     A=&MergeHelper_VectorSwitch(
      0=Linked() {linkedVariable {name=Particles.ShapeLocation.ShapePosition, type=NiagaraPosition}}
      1=Linked() {linkedVariable {name=Particles.ShapeLocation001.ShapePosition, type=NiagaraPosition}}
      2=Linked() {linkedVariable {name=Particles.ShapeLocation002.ShapePosition, type=NiagaraPosition}}
      3=?
      Index=&SpawnGroup()
      Num=::NewEnumerator1
     )
     B=&SelectVectorFromArray(
      Direct Array Index=&ReturnExecIndex()
      Vector Selection Array=DataInterface() {propertyValues {FloatData [
         <0 0 0>
         <4.9957 0 0>
         <13.5687 8.0034 0>
         <7.7255 -11.1569 0>
         <-14.792 -5.3706 0>
         <0 0 0>
         <7.3301 -17.4146 0>
         <-1.8693 -0.8973 0>
         <0 0 0>
         <2.779 -17.0026 0>
         <6.4519 9.7094 0>
         <-2.0611 -2.8459 0>
         <1.1209 0 0>
         <-5.647 0 0>
         <0.5993 10.8353 0>
         <0 0 0>
         <0 0 0>
         <0 0 0>
         <0 -2.7014 0>
         <0 0 0>
         <5.1262 8.5522 0>
         <0 0 0>
         <0 0 0>
         <0 0 0>
         <0 0 0>
         <0 0 0>
        ]}}
     )
    )
   )
   &InitialMeshOrientation(
    Mesh Axis To Orient=&RandomRangeVector(Minimum=<0 1 -0.5>, Maximum=<0 1 -0.2>)
    Mesh Orientation Mode=::NewEnumerator2
    Orientation Vector=&SubtractVectorFromPosition(Position=$Engine.Emitter.SimulationPosition, Vector=Linked() {linkedVariable {name=Particles.Position, type=NiagaraPosition}})
    Use Rotation Vector=false
   )
   /SetVariables(
    Particles.Rand001:NiagaraFloat=&RandomRangeFloat(Minimum=1, Maximum=1.2)
    Particles.Rand002:NiagaraFloat=&RandomRangeFloat(Minimum=0, Maximum=-3)
   ) {isSetParameters=true}
  ]
  particleUpdateScript [
   &ParticleState(Kill Particles When Lifetime Has Elapsed=false)
   &ScaleMeshSize(Scale Factor=&MakeVector(
     X=1
     Y=1
     Z=&FloatFromCurve(
      FloatCurve=DataInterface() {propertyValues {Curve {
         defaultValue=1
         keys [
          {arriveTangent=22.3836, interpMode=RCIM_Cubic, leaveTangent=22.3836, tangentMode=RCTM_User, time=0, value=0}
          {arriveTangentWeight=0.1608, interpMode=RCIM_Cubic, leaveTangentWeight=0.1828, tangentMode=RCTM_User, time=0.0771, value=1.1005}
          {interpMode=RCIM_Cubic, tangentMode=RCTM_User, time=0.1223, value=1}
          <at 0.3943 1 RCIM_Cubic>
         ]
        }}}
      CurveIndex=&Add_Float(A=&Multiply_Float(A=$Particles.NormalizedAge, B=$Particles.Rand001), B=$Particles.Rand002)
     )
    ))
   /SetVariables(Particles.Position=/Niagara/DynamicInputs/TypeConversions/MakeCustomPositionFromBool(
     Bool:NiagaraBool=&SetBoolByFloatComparison(
      A=&FloatFromCurve(
       FloatCurve=DataInterface() {propertyValues {Curve {
          defaultValue=1
          keys [
           {arriveTangent=22.3836, interpMode=RCIM_Cubic, leaveTangent=22.3836, tangentMode=RCTM_User, time=0, value=0}
           {arriveTangentWeight=0.1608, interpMode=RCIM_Cubic, leaveTangentWeight=0.1828, tangentMode=RCTM_User, time=0.0771, value=1.1005}
           {interpMode=RCIM_Cubic, tangentMode=RCTM_User, time=0.1223, value=1}
           <at 0.3943 1 RCIM_Cubic>
          ]
         }}}
       CurveIndex=&Add_Float(A=&Multiply_Float(A=$Particles.NormalizedAge, B=$Particles.Rand001), B=$Particles.Rand002)
      )
      B=0.05
     )
     True Position:NiagaraPosition=$Particles.Initial.Position
     False Position:NiagaraPosition=<0 0 0>
    )) {isSetParameters=true}
  ]
  rendererClasses [&NiagaraMeshRendererProperties]
  renderers [{rendererClass=&NiagaraMeshRendererProperties, rendererIndex=0}]
 }
 Swords001 {
  bEnabled=false
  emitterUpdateScript [
   &EmitterState(
    Inactive Response=::NewEnumerator1
    Life Cycle Mode=::NewEnumerator1
    Loop Behavior=::NewEnumerator1
    Loop Delay=3.5
    Loop Duration Mode=::NewEnumerator1
    UseLoopDelay=true
   )
   &SpawnBurst_Instantaneous(Spawn Count=14, Spawn Time=0.1, Use Spawn Probability=false, UseLoopCountLimit=false) {module=SpawnBurst_Instantaneous001}
   &SpawnBurst_Instantaneous(Spawn Count=12, Use Spawn Probability=false, UseLoopCountLimit=false)
  ]
  particleSpawnScript [
   &FN_SetSpawnGroup(Number of Spawn Groups=3)
   &InitializeParticle(Color Mode=::NewEnumerator0, Mesh Scale Max=<1 1 0.2>, Mesh Scale Mode=::NewEnumerator2, UsePositionOffset=false, Write Material Random=true)
   &ShapeLocation(
    Apply To Particle Position=::NewEnumerator1
    Box Size=<25 90 1>
    MaskSpawnGroup=false
    Offset=<-52 0 0>
    Offset Mode=::NewEnumerator0
    Override Seed=false
    Rotation Mode=::NewEnumerator2
    Shape Primitive=::NewEnumerator2
    Surface Only=false
    UseOwnerScale=true
    Yaw / Pitch / Roll=<0.0427 0 0>
   )
   &ShapeLocation(
    Apply To Particle Position=::NewEnumerator1
    Box Size=<70 15 1>
    MaskSpawnGroup=false
    Offset=<0 -50 0>
    Offset Mode=::NewEnumerator0
    Override Seed=false
    Shape Primitive=::NewEnumerator2
    Surface Only=false
    UseOwnerScale=true
   ) {module=ShapeLocation001}
   &ShapeLocation(
    Apply To Particle Position=::NewEnumerator1
    Box Size=<25 90 1>
    MaskSpawnGroup=false
    Offset=<52 0 0>
    Offset Mode=::NewEnumerator0
    Override Seed=false
    Rotation Mode=::NewEnumerator2
    Shape Primitive=::NewEnumerator2
    Surface Only=false
    UseOwnerScale=true
    Yaw / Pitch / Roll=<-0.0427 0 0>
   ) {module=ShapeLocation002}
   &SystemLocation(
    MaskSpawnGroup=false
    Offset=&MergeHelper_VectorSwitch(
     0=Linked() {linkedVariable {name=Particles.ShapeLocation.ShapePosition, type=NiagaraPosition}}
     1=Linked() {linkedVariable {name=Particles.ShapeLocation001.ShapePosition, type=NiagaraPosition}}
     2=Linked() {linkedVariable {name=Particles.ShapeLocation002.ShapePosition, type=NiagaraPosition}}
     3=?
     Index=&SpawnGroup()
     Num=::NewEnumerator1
    )
   )
   &InitialMeshOrientation(
    Mesh Axis To Orient=&RandomRangeVector(Minimum=<0 1 -0.5>, Maximum=<0 1 -0.2>)
    Mesh Orientation Mode=::NewEnumerator2
    Orientation Vector=&SubtractVectorFromPosition(Position=$Engine.Emitter.SimulationPosition, Vector=Linked() {linkedVariable {name=Particles.Position, type=NiagaraPosition}})
    Use Rotation Vector=false
   )
  ]
  particleUpdateScript [
   &ParticleState(Kill Particles When Lifetime Has Elapsed=false)
   &ScaleMeshSize(Scale Factor=&MakeVector(
     X=1
     Y=1
     Z=&FloatFromCurve(
      FloatCurve=DataInterface() {propertyValues {Curve {
         defaultValue=1
         keys [
          {arriveTangent=22.3836, interpMode=RCIM_Cubic, leaveTangent=22.3836, tangentMode=RCTM_User, time=0, value=0}
          {arriveTangentWeight=0.1608, interpMode=RCIM_Cubic, leaveTangentWeight=0.1828, tangentMode=RCTM_User, time=0.1031, value=1.3731}
          {interpMode=RCIM_Cubic, tangentMode=RCTM_User, time=0.2348, value=0.8092}
          <at 0.3943 1 RCIM_Cubic>
         ]
        }}}
      CurveIndex=&Multiply_Float(A=$Particles.NormalizedAge, B=&RandomRangeFloat(Minimum=1, Maximum=1.6))
     )
    ))
  ]
  rendererClasses [&NiagaraMeshRendererProperties]
  renderers [{rendererClass=&NiagaraMeshRendererProperties, rendererIndex=0}]
 }
}
summary {
 systemName=NS_Emote_CurbRestore_Spawn
 emitters [
  {bEnabled=false, emitterName=Swords, rendererClasses [&NiagaraMeshRendererProperties]}
  {emitterName=Debris, rendererClasses [&NiagaraMeshRendererProperties, &NiagaraSpriteRendererProperties]}
  {emitterName=Dust, rendererClasses [&NiagaraSpriteRendererProperties]}
  {bEnabled=false, emitterName=DustBurst, rendererClasses [&NiagaraSpriteRendererProperties]}
  {
   bEnabled=false
   emitterName=Debris001
   rendererClasses [&NiagaraMeshRendererProperties, &NiagaraSpriteRendererProperties]
  }
  {bEnabled=false, emitterName=Swords001, rendererClasses [&NiagaraMeshRendererProperties]}
 ]
 userVariables []
}
